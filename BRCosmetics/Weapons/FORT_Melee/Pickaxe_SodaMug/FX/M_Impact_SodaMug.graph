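# Translucent unlit Niagara sprite material for the Soda Mug pickaxe impact effect.
# Emissive starts as boosted particle color easing over lifetime (plain color on mobile), with an optional RetroFX blend to time-cycled rainbow colors driven by screen-aligned noise.
# An impact flash texture supplies a mask whose erosion and fade follow particle alpha.
# An offset from the particle center, disabled on mobile, scales with camera distance to resize distant sprites.
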Material /BRCosmetics/Weapons/FORT_Melee/Pickaxe_SodaMug/FX/M_Impact_SodaMug
graph {
 properties {bUsedWithNiagaraSprites=true, blendMode=BLEND_Translucent, pixelDepthOffsetMode=PDOM_Legacy, shadingModel=MSM_Unlit, usedShadingModels=Unlit}
 nodes [
  n1 = ParticleColor()
  n2 = LinearInterpolate(
   A=Multiply(A=ScalarParameter() {defaultValue=4, parameterName="Particle color start  Multiply"}, B=@n1.RGB) {constA=4}
   B=@n1.RGB
   Alpha=ParticleRelativeTime()
  ) {constA=6}
  n3 = TextureSampleParameter2D(UVs=MaterialFunctionCall() {materialFunction=&ScreenAlignedUVs}."X 100%, Y 100%") {parameterName=RetroFX_NoiseTexture, texture=/Game/Global/Textures/Default/Masks_and_Gradients/chunkma_2}
  n4 = Time()
  n5 = MaterialFunctionCall!(Gradient=Frac(None=Add(A=@n3.R, B=@n4))) {materialFunction=/BRCosmetics/Athena/Effects/Particles/Skydiving/NS_Contrail_DarkAndLight/MF_DarkAndLight_RainbowColors}
  MaterialFunctionCall(Emissive=LinearInterpolate(
    A=ShadingPathSwitch(Default=@n2, Mobile=@n1.RGB)
    B=ComponentMask(None=@n5.Result) {b=true, g=true, r=true}
    Alpha=ScalarParameter() {parameterName=RetroFX_Intensity}
   )) {materialFunction=&MF_FX_EmissiveAdjustments}
  n6 = TextureSampleParameter2D() {parameterName=ImpactTexture, texture=/Game/Effects/Fort_Effects/Textures/Explo/T_Impact_Flash_2x2_01}
  n7 = Add(A=LinearInterpolate(Alpha=@n1.A) {constA=-0.6, constB=0.6}, B=@n6.B)
  Multiply(A=@n1.A, B=Clamp(None=Multiply(A=@n7, B=@n6.G)))
  n8 = ParticlePositionWS()
  n9 = Subtract(A=@n8, B=WorldPosition().XYZ)
  n10 = ScalarParameter() {defaultValue=5, parameterName=MaxSize}
  n11 = CameraPositionWS()
  n12 = Divide(
   A=Add(A=Distance(A=@n8, B=@n11), B=ScalarParameter() {defaultValue=-1250, parameterName=ScaleDistanceOffset})
   B=ScalarParameter() {defaultValue=20000, parameterName=Scale Distance}
  )
  ShadingPathSwitch(Default=Multiply(A=Multiply(A=@n9, B=@n10), B=Clamp(None=@n12)), Mobile=Constant())
 ]
}
