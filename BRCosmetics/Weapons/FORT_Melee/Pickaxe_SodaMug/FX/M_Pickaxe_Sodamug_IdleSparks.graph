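# Translucent unlit Niagara mesh-particle material for the idle sparks on the Soda Mug pickaxe.
# Particle color gets hue, brightness and desaturation control from dynamic parameters, then blends toward the unadjusted color along a linear gradient.
# An optional RetroFX layer blends in time-cycled rainbow colors driven by screen-aligned noise, before shared emissive adjustments.
# Particle alpha, partly shaped by a Fresnel term, feeds a depth fade for soft intersections.
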Material /BRCosmetics/Weapons/FORT_Melee/Pickaxe_SodaMug/FX/M_Pickaxe_Sodamug_IdleSparks
graph {
 properties {
  bAutomaticallySetUsageInEditor=false
  bUsedWithNiagaraMeshParticles=true
  blendMode=BLEND_Translucent
  pixelDepthOffsetMode=PDOM_Legacy
  previewMesh=&/SM_FX_SingleSpike
  shadingModel=MSM_Unlit
  usedShadingModels=Unlit
 }
 nodes [
  ComponentMask() {g=true, r=true}
  n1 = ParticleColor()
  n2 = MaterialFunctionCall(
   Invert Fresnel=StaticBoolParameter() {defaultValue=true, parameterName=Invert Fresnel}
   Power=ScalarParameter() {defaultValue=1, parameterName=Fresnel Power}
  ) {materialFunction=&Fresnel_Function}
  DepthFade(
   Opacity=LinearInterpolate(A=@n1.A, B=Multiply(A=@n1.A, B=@n2.Result), Alpha=ScalarParameter() {defaultValue=0.5, parameterName=useFresnel?})
   FadeDistance=ScalarParameter() {defaultValue=4, parameterName=FadeDistance}
  )
  n3 = DynamicParameter() {defaultValue=<rgba 0.05 0.3 0 1>, paramNames [HueShiftPercent, Brightness, DeSaturation, Param4]}
  n4 = LinearInterpolate(
   A=ParticleColor().RGB
   B=VectorParameter() {parameterName=v, defaultValue=<rgba 1 0 0.0754 1>}.RGB
   Alpha=ScalarParameter() {parameterName=DebugColor}
  )
  n5 = Desaturation(
   None=MaterialFunctionCall(Hue Shift Percentage=@n3.HueShiftPercent, Texture=Multiply(A=@n4, B=@n3.Brightness)) {materialFunction=&HueShift}.Result
   Fraction=@n3.DeSaturation
  )
  n6 = MaterialFunctionCall(
   GradientStart=ScalarParameter() {parameterName=TipColorBias}
   RotationAngle=ScalarParameter() {defaultValue=0.25, parameterName=Param_2}
   GradientOffset=ScalarParameter() {defaultValue=0.2065, parameterName="GradientOffset (S)"}
  ) {materialFunction=&MF_UI_LinearGradient}
  n7 = MaterialFunctionCall() {materialFunction=&ScreenAlignedUVs}
  n8 = TextureSampleParameter2D(UVs=@n7."X 100%, Y 100%") {parameterName=RetroFX_NoiseTexture, texture=/Game/Global/Textures/Default/Masks_and_Gradients/chunkma_2}
  n9 = Time()
  n10 = MaterialFunctionCall!(Gradient=Frac(None=Add(A=@n8.R, B=@n9))) {materialFunction=/BRCosmetics/Athena/Effects/Particles/Skydiving/NS_Contrail_DarkAndLight/MF_DarkAndLight_RainbowColors}
  MaterialFunctionCall(Emissive=LinearInterpolate(
    A=LinearInterpolate(A=@n5, B=@n4, Alpha=@n6.Result)
    B=ComponentMask(None=@n10.Result) {b=true, g=true, r=true}
    Alpha=ScalarParameter() {parameterName=RetroFX_Intensity}
   )) {materialFunction=&MF_FX_EmissiveAdjustments}
  Panner(Coordinate=@n7."X 100%, Y 100%") {speedX=0.5, speedY=0.5}
  ScalarParameter() {defaultValue=1}
 ]
}
