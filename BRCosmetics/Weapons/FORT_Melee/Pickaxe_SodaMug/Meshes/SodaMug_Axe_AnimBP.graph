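# Animation Blueprint for the Soda Mug pickaxe that adds secondary physics motion to its three dyn_logo bones.
# The cached input pose feeds two chains of Anim Dynamics nodes with different tuning, blended by bIsInFrontEnd.
# The front-end chain uses external forces and stronger gravity, while the other responds more to component movement.
# The update event maps delta time to FPSAlpha, which the AnimGraph never reads.
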AnimBlueprint /BRCosmetics/Weapons/FORT_Melee/Pickaxe_SodaMug/Meshes/SodaMug_Axe_AnimBP
target_skeleton=&/SodaMug_Axe_Skeleton
graphs [
 AnimGraph {nodes [
   n1 = Anim Dynamics - Bone: dyn_logo_1(ComponentPose=Local To Component(LocalPose=Use cached pose 'InputPoseCache'()), Alpha=1, ExternalForce="-200,-200,0") {
    angularDampingOverride=0.999
    bOverrideAngularDamping=true
    bUseGravityOverride=true
    alphaBoolBlend {blendOption=Linear}
    boundBone {boneName=dyn_logo_1}
    componentAppliedLinearAccClamp=<1000 1000 1000>
    componentLinearAccScale=<0.1 0.1 0.1>
    componentLinearVelScale=<0.1 0.1 0.1>
    gravityOverride {z=-980}
    physicsBodyDefinitions [{
      boundBone {boneName=dyn_logo_1}
      boxExtents=<5 10 10>
      constraintSetup {angularLimitsMax=<0 30 20>, angularLimitsMin=<0 -10 -20>, angularTarget=<0 0 0>}
      localJointOffset=<3.5 0 0>
     }]
   }
   n2 = Anim Dynamics - Bone: dyn_logo_2(ComponentPose=@n1, Alpha=1, ExternalForce="0,400,0") {
    angularDampingOverride=0.999
    bOverrideAngularDamping=true
    bUseGravityOverride=true
    alphaBoolBlend {blendOption=Linear}
    boundBone {boneName=dyn_logo_2}
    componentAppliedLinearAccClamp=<1000 1000 1000>
    componentLinearAccScale=<0.1 0.1 0.1>
    componentLinearVelScale=<0.1 0.1 0.1>
    gravityOverride {z=-980}
    physicsBodyDefinitions [{
      boundBone {boneName=dyn_logo_2}
      boxExtents=<5 10 10>
      constraintSetup {angularLimitsMax=<5 40 25>, angularLimitsMin=<-5 -10 -25>, angularTarget=<0 0 0>}
      localJointOffset=<3 0 0>
     }]
   }
   n3 = Anim Dynamics - Bone: dyn_logo_3(ComponentPose=@n2, Alpha=1, ExternalForce="400,0,0") {
    angularDampingOverride=0.999
    bOverrideAngularDamping=true
    bUseGravityOverride=true
    alphaBoolBlend {blendOption=Linear}
    boundBone {boneName=dyn_logo_3}
    componentAppliedLinearAccClamp=<1000 1000 1000>
    componentLinearAccScale=<0.1 0.1 0.1>
    componentLinearVelScale=<0.1 0.1 0.1>
    gravityOverride {z=-980}
    physicsBodyDefinitions [{
      boundBone {boneName=dyn_logo_3}
      boxExtents=<5 10 10>
      constraintSetup {angularLimitsMax=<0 30 20>, angularLimitsMin=<0 -5 -20>, angularTarget=<0 0 0>}
      localJointOffset=<2.5 0 0>
     }]
   }
   n4 = Anim Dynamics - Bone: dyn_logo_1(ComponentPose=Local To Component(LocalPose=Use cached pose 'InputPoseCache'()), Alpha=1, ExternalForce="0,0,0") {
    angularDampingOverride=0.999
    bOverrideAngularDamping=true
    bUseGravityOverride=true
    alphaBoolBlend {blendOption=Linear}
    boundBone {boneName=dyn_logo_1}
    componentAppliedLinearAccClamp=<4000 3200 4000>
    componentLinearAccScale=<0.35 0.25 0.35>
    componentLinearVelScale=<0.25 0.2 0.25>
    gravityOverride {z=-400}
    physicsBodyDefinitions [{
      boundBone {boneName=dyn_logo_1}
      boxExtents=<5 10 10>
      constraintSetup {angularLimitsMax=<0 15 30>, angularLimitsMin=<0 -15 -30>, angularTarget=<0 0 0>}
      localJointOffset=<4.5 0 0>
     }]
   }
   n5 = Anim Dynamics - Bone: dyn_logo_2(ComponentPose=@n4, Alpha=1, ExternalForce="0,0,0") {
    angularDampingOverride=0.999
    bOverrideAngularDamping=true
    bUseGravityOverride=true
    alphaBoolBlend {blendOption=Linear}
    boundBone {boneName=dyn_logo_2}
    componentAppliedLinearAccClamp=<4000 3200 4000>
    componentLinearAccScale=<0.35 0.25 0.35>
    componentLinearVelScale=<0.25 0.2 0.25>
    gravityOverride {z=-500}
    physicsBodyDefinitions [{
      boundBone {boneName=dyn_logo_2}
      boxExtents=<5 10 10>
      constraintSetup {angularLimitsMax=<10 20 35>, angularLimitsMin=<-10 -20 -35>, angularTarget=<0 0 0>}
      localJointOffset=<4 0 0>
     }]
   }
   n6 = Anim Dynamics - Bone: dyn_logo_3(ComponentPose=@n5, Alpha=1, ExternalForce="0,0,0") {
    angularDampingOverride=0.999
    bOverrideAngularDamping=true
    bUseGravityOverride=true
    alphaBoolBlend {blendOption=Linear}
    boundBone {boneName=dyn_logo_3}
    componentAppliedLinearAccClamp=<4000 3200 4000>
    componentLinearAccScale=<0.35 0.25 0.35>
    componentLinearVelScale=<0.25 0.2 0.25>
    gravityOverride {z=-600}
    physicsBodyDefinitions [{
      boundBone {boneName=dyn_logo_3}
      boxExtents=<5 10 10>
      constraintSetup {angularLimitsMax=<1 15 30>, angularLimitsMin=<-1 -15 -29>, angularTarget=<0 0 0>}
      localJointOffset=<3.5 0 0>
     }]
   }
   Output Pose(Result=Blend Poses by bool(BlendPose_0=Component To Local(ComponentPose=@n3), BlendPose_1=Component To Local(ComponentPose=@n6), bActiveValue=Get bIsInFrontEnd()) {blendType=Linear}) {layerGroup=None}
   AnimGraphNode_SubInput_0 = Input Pose()
   Save cached pose 'InputPoseCache'(Pose=@AnimGraphNode_SubInput_0) {cacheName=InputPoseCache}
  ]}
 EventGraph {nodes [
   n1 = Event BlueprintUpdateAnimation()
   > Set FPSAlpha(FPSAlpha:double=MapRangeClamped(Value=@n1.DeltaTimeX, InRangeA=0.033, InRangeB=0.016, OutRangeA=1)) -> (Output_Get:double)
   TryGetPawnOwner()
  ]}
]
variables [FPSAlpha {type=number}]
